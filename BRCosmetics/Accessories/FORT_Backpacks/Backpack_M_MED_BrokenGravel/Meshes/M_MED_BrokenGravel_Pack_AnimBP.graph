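# Animation Blueprint for the BrokenGravel backpack cosmetic, posing the pack and adding secondary motion to its charms, ropes and gun.
# AnimGraph applies bone offsets to pack, rope and charm bones, caches that pose, then runs chained root-relative, angular-spring Anim Dynamics nodes.
# EventGraph computes pack and cable offsets from spine yaw, flags BrokenGravel skeletal meshes, and sets gravity and charm variables for front-end, skydiving and DBNO states.
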AnimBlueprint /BRCosmetics/Accessories/FORT_Backpacks/Backpack_M_MED_BrokenGravel/Meshes/M_MED_BrokenGravel_Pack_AnimBP
target_skeleton=&/M_MED_BrokenGravel_Pack_Skeleton
graphs [
 AnimGraph {nodes [
   Output Pose(Result=Use cached pose 'AnimDynamics'()) {layerGroup=None}
   n1 = "Transform (Modify) Bone - Bone: pack"!(ComponentPose=Local To Component(LocalPose=Input Pose()), Alpha, Translation="0,5,-5", Rotation="0,0,0") -> (Pose) {rotationMode=BMM_Additive, scaleMode=BMM_Additive, translationMode=BMM_Additive, boneToModify {boneName=pack}, translation {y=5, z=-5}}
   n2 = "Transform (Modify) Bone - Bone: pack"!(
    ComponentPose="Transform (Modify) Bone - Bone: base_ropes_01"!(ComponentPose=@n1, Alpha=1, Rotation="0,0,50") -> (Pose) {rotationMode=BMM_Additive, scaleMode=BMM_Additive, boneToModify {boneName=base_ropes_01}, rotation {roll=50}}
    bAlphaBoolEnabled:bool
    Translation="-2,0,0"
    Rotation="0,0,15"
   ) -> (Pose) {
    alphaInputType=Bool
    rotationMode=BMM_Replace
    scaleMode=BMM_Replace
    translationMode=BMM_Additive
    alphaBoolBlend {blendOption=Linear}
    boneToModify {boneName=pack}
    rotation {roll=15}
    translation {x=-2}
   }
   Save cached pose 'Transforms'(Pose=Component To Local(ComponentPose="Transform (Modify) Bone - Bone: base_charm_01"!(ComponentPose=@n2, Alpha=1, Rotation="15,0,0") -> (Pose) {rotationMode=BMM_Additive, scaleMode=BMM_Additive, boneToModify {boneName=base_charm_01}, rotation {pitch=15}})) {cacheName=Transforms}
   n3 = Anim Dynamics - Bone: dyn_charm_01!(ComponentPose=Local To Component(LocalPose=Use cached pose 'Transforms'()), Alpha=1, GravityOverride:Vector="0,0,-980") -> (Pose) {
    angularDampingOverride=1
    angularSpringConstant=50
    bAngularSpring=true
    bOverrideAngularDamping=true
    bUseGravityOverride=true
    gravityScale=0.35
    simulationSpace=RootRelative
    boundBone {boneName=dyn_charm_01}
    componentAppliedLinearAccClamp=<3000 3000 3000>
    componentLinearAccScale=<0.4 0.4 0.4>
    componentLinearVelScale=<0.2 0.2 0.2>
    gravityOverride {z=-980}
    physicsBodyDefinitions [{
      boundBone {boneName=dyn_charm_01}
      boxExtents=<15 5 15>
      constraintSetup {angularLimitsMax=<0 20 15>, angularLimitsMin=<0 -20 -70>, angularTarget=<1 0 0>}
      localJointOffset=<4 0 0>
     }]
    simSpaceSettings {maxAngularAcceleration=20, maxAngularVelocity=10, simSpaceAngularAlpha=1}
   }
   n4 = Anim Dynamics - Bone: dyn_charm_02!(ComponentPose=@n3, Alpha=1, GravityOverride:Vector="0,0,-980") -> (Pose) {
    angularDampingOverride=1
    angularSpringConstant=50
    bAngularSpring=true
    bOverrideAngularDamping=true
    bUseGravityOverride=true
    gravityScale=0.35
    simulationSpace=RootRelative
    boundBone {boneName=dyn_charm_02}
    componentAppliedLinearAccClamp=<3000 3000 3000>
    componentLinearAccScale=<0.4 0.4 0.4>
    componentLinearVelScale=<0.4 0.4 0.4>
    gravityOverride {z=-980}
    physicsBodyDefinitions [{
      boundBone {boneName=dyn_charm_02}
      boxExtents=<20 20 8>
      constraintSetup {angularLimitsMax=<0 40 20>, angularLimitsMin=<0 -50 -40>, angularTarget=<1 0 0>}
      localJointOffset=<6 0 0>
     }]
    simSpaceSettings {maxAngularAcceleration=20, maxAngularVelocity=10, simSpaceAngularAlpha=1}
   }
   n5 = Anim Dynamics - Bone: dyn_ropes_01!(ComponentPose=@n4, Alpha=0.8, GravityOverride:Vector="0,0,-980") -> (Pose) {
    angularDampingOverride=1
    angularSpringConstant=9000
    bAngularSpring=true
    bOverrideAngularDamping=true
    bUseGravityOverride=true
    gravityScale=0.35
    simulationSpace=RootRelative
    boundBone {boneName=dyn_ropes_01}
    componentAppliedLinearAccClamp=<3000 3000 3000>
    componentLinearAccScale=<0.4 0.4 0.4>
    componentLinearVelScale=<0.4 0.4 0.4>
    gravityOverride {z=-980}
    physicsBodyDefinitions [{
      boundBone {boneName=dyn_ropes_01}
      boxExtents=<20 20 20>
      constraintSetup {angularLimitsMax=<0 20 70>, angularLimitsMin=<0 -20 -50>, angularTarget=<1 0 0>}
      localJointOffset=<4 0 0>
     }]
    simSpaceSettings {maxAngularAcceleration=20, maxAngularVelocity=10, simSpaceAngularAlpha=1}
   }
   n6 = Anim Dynamics - Bone: dyn_ropes_a_02!(ComponentPose=@n5, Alpha=0.8, GravityOverride:Vector="0,0,-980") -> (Pose) {
    angularDampingOverride=1
    angularSpringConstant=9000
    bAngularSpring=true
    bOverrideAngularDamping=true
    bUseGravityOverride=true
    gravityScale=0.35
    simulationSpace=RootRelative
    boundBone {boneName=dyn_ropes_a_02}
    componentAppliedLinearAccClamp=<3000 3000 3000>
    componentLinearAccScale=<0.4 0.4 0.4>
    componentLinearVelScale=<0.4 0.4 0.4>
    gravityOverride {z=-980}
    physicsBodyDefinitions [{
      boundBone {boneName=dyn_ropes_a_02}
      boxExtents=<20 20 20>
      constraintSetup {angularLimitsMax=<0 20 0>, angularLimitsMin=<0 -20 -70>, angularTarget=<1 0 0>}
      localJointOffset=<6 0 0>
     }]
    simSpaceSettings {maxAngularAcceleration=20, maxAngularVelocity=10, simSpaceAngularAlpha=1}
   }
   n7 = Anim Dynamics - Bone: dyn_ropes_c_02!(ComponentPose=@n6, Alpha=0.8, GravityOverride:Vector="0,0,-980") -> (Pose) {
    angularDampingOverride=1
    angularSpringConstant=9000
    bAngularSpring=true
    bOverrideAngularDamping=true
    bUseGravityOverride=true
    gravityScale=0.35
    simulationSpace=RootRelative
    boundBone {boneName=dyn_ropes_c_02}
    componentAppliedLinearAccClamp=<3000 3000 3000>
    componentLinearAccScale=<0.4 0.4 0.4>
    componentLinearVelScale=<0.4 0.4 0.4>
    gravityOverride {z=-980}
    physicsBodyDefinitions [{
      boundBone {boneName=dyn_ropes_c_02}
      boxExtents=<20 20 20>
      constraintSetup {angularLimitsMax=<0 20 10>, angularLimitsMin=<0 -20 -20>, angularTarget=<1 0 0>}
      localJointOffset=<4 0 0>
     }]
    simSpaceSettings {maxAngularAcceleration=20, maxAngularVelocity=10, simSpaceAngularAlpha=1}
   }
   n8 = Anim Dynamics - Bone: dyn_ropes_c_02(ComponentPose=@n7, Alpha=0.8, GravityOverride="0,0,-980") {
    angularDampingOverride=1
    angularSpringConstant=9000
    bAngularSpring=true
    bOverrideAngularDamping=true
    bUseGravityOverride=true
    gravityScale=0.35
    simulationSpace=RootRelative
    boundBone {boneName=dyn_ropes_c_02}
    componentAppliedLinearAccClamp=<3000 3000 3000>
    componentLinearAccScale=<0.4 0.4 0.4>
    componentLinearVelScale=<0.4 0.4 0.4>
    gravityOverride {z=-980}
    physicsBodyDefinitions [{
      boundBone {boneName=dyn_ropes_c_02}
      boxExtents=<20 20 20>
      constraintSetup {angularLimitsMax=<0 20 5>, angularLimitsMin=<0 -20 -10>, angularTarget=<1 0 0>}
      localJointOffset=<2 0 0>
     }]
    simSpaceSettings {maxAngularAcceleration=20, maxAngularVelocity=10, simSpaceAngularAlpha=1}
   }
   n9 = Anim Dynamics - Bone: dyn_gun_01!(ComponentPose=@n8, Alpha=1, GravityOverride:Vector="0,0,-980") -> (Pose) {
    angularDampingOverride=1
    angularSpringConstant=15000
    bAngularSpring=true
    bOverrideAngularDamping=true
    bUseGravityOverride=true
    gravityScale=0.35
    simulationSpace=RootRelative
    boundBone {boneName=dyn_gun_01}
    componentAppliedLinearAccClamp=<3000 3000 3000>
    componentLinearAccScale=<0.4 0.4 0.4>
    componentLinearVelScale=<0.4 0.4 0.4>
    gravityOverride {z=-980}
    physicsBodyDefinitions [{
      boundBone {boneName=dyn_gun_01}
      boxExtents=<35 35 35>
      constraintSetup {angularLimitsMax=<2 0 10>, angularLimitsMin=<-2 -5 -5>, angularTarget=<1 0 0>}
      localJointOffset=<3 0 0>
     }]
    simSpaceSettings {maxAngularAcceleration=20, maxAngularVelocity=10, simSpaceAngularAlpha=1}
   }
   Save cached pose 'AnimDynamics'(Pose=Component To Local(ComponentPose=@n9)) {cacheName=AnimDynamics}
  ]}
 EventGraph {nodes [
   Event BlueprintUpdateAnimation()
   n1 = GetOwnerSkeletalMeshComponent()
   n2 = BreakRotator(InRot=BreakTransform(InTransform=GetDeltaTransformFromRefPose(self=@n1, BoneName=spine_03, BaseName=spine_05)).Rotation)
   n3 = MakeRotator(Roll=MapRangeClamped(Value=@n2.Yaw, InRangeA=0, InRangeB=15, OutRangeB=20))
   > Set backpackRotate(backpackRotate:Rotator=@n3) -> (Output_Get:Rotator)
   n4 = MapRangeClamped(Value=@n2.Yaw, InRangeA=0, InRangeB=15, OutRangeB=-1)
   > Set backpackTranslate(backpackTranslate:Vector=MakeVector(Y=float + float(A=@n4, B=-1), Z=3)) -> (Output_Get:Vector)
   > Set cablesRotate(cablesRotate:Rotator=MakeRotator(Roll=MapRangeClamped(Value=@n2.Yaw, InRangeA=0, InRangeB=15, OutRangeB=50))) -> (Output_Get:Rotator)
   > n5 = Sequence!(execute) -> (then_0, then_1, then_2:Exec)
   > n6 = Cast To FortPlayerPawn(Object=TryGetPawnOwner())
   n7 = GetDisplayName(Object=GetSkeletalMeshForPartType(self=@n6))
   > Set isBrokenGravelAlpha(isBrokenGravelAlpha:double="To Float (Boolean)"(InBool=MatchesWildcard(SourceString=@n7, Wildcard=*BrokenGravel*))) -> (Output_Get:double)
   n8 = Branch(execute=@n5.then_2, Condition=Get bIsInFrontEnd())
   > Set AlphaFECharms(AlphaFECharms:double=0.3) -> (Output_Get:double)
   > Set Override World Gravity(Override World Gravity:Vector="0,0,-1800") -> (Output_Get:Vector)
   > Set backpackTranslate(backpackTranslate=vector + vector(A="3,0,0", B=Get backpackTranslate() -> (backpackTranslate:Vector)))
   > Set CharmPose(CharmPose:Rotator="20,0,0") -> (Output_Get:Rotator)
   Set AlphaFECharms(execute=@n8.else, AlphaFECharms=1)
   > n9 = Branch(Condition=Get bIsDBNO())
   > Set CharmPose(CharmPose="20,0,-30")
   Set CharmPose(execute=@n9.else, CharmPose="0, 0, 0")
   n10 = Branch(execute=@n5.then_1, Condition=Get bIsSkyDiving())
   > Set Override World Gravity(Override World Gravity="0,0,-350")
   Set Override World Gravity(execute=@n10.else, Override World Gravity="0,0,-980")
   Reroute Node(InputPin:Rotator=@n3) -> (OutputPin:Rotator)
  ]}
]
variables [
 AlphaFECharms {type=number}
 backpackRotate {type=Rotator}
 backpackTranslate {type=Vector}
 cablesRotate {type=Rotator}
 CharmPose {type=Rotator}
 isBrokenGravelAlpha {type=number}
 Override World Gravity {type=Vector}
]
